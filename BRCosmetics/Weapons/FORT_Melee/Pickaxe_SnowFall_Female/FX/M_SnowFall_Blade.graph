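# Translucent Niagara material for blade effects on the SnowFall pickaxe, usable on sprites, ribbons and meshes.
# Dynamic-parameter colors blend through a curve-atlas row or two panning noises, with dimmer back faces; color green channels reuse red.
# Opacity combines gradients, camera, view-angle and depth fades, a facing dot product, and optional particle-driven noise erosion.
# Panning noise drives world position offset, with a static switch choosing UV or instance-space projection.
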Material /BRCosmetics/Weapons/FORT_Melee/Pickaxe_SnowFall_Female/FX/M_SnowFall_Blade
graph {
 properties {
  bUsedWithNiagaraMeshParticles=true
  bUsedWithNiagaraRibbons=true
  bUsedWithNiagaraSprites=true
  blendMode=BLEND_Translucent
  pixelDepthOffsetMode=PDOM_Legacy
  translucencyLightingMode=TLM_VolumetricPerVertexDirectional
  usedShadingModels=Default Lit
 }
 nodes [
  Constant() {r=1}
  n1 = DynamicParameter() {parameterIndex=1, paramNames [SecondaryR,SecondaryG,SecondaryB,SecondaryA]}
  n2 = AppendVector(A=AppendVector(A=AppendVector(A=@n1.SecondaryR, B=@n1.SecondaryR), B=@n1.SecondaryB), B=@n1.SecondaryA)
  n3 = Multiply(A=@n2, B=ScalarParameter() {defaultValue=0.08, group=Color, parameterName=Backside Color Intensity})
  n4 = ScalarParameter() {defaultValue=0.3, group=Color, parameterName=Backside Color Desaturation}
  n5 = TextureCoordinate()
  n6 = AppendVector(
   A=ScalarParameter() {defaultValue=1.5, group=Noise, parameterName=NoiseTexture_Tile_U}
   B=ScalarParameter() {defaultValue=0.5, group=Noise, parameterName=NoiseTexture_Tile_V}
  )
  n7 = ScalarParameter() {group=Noise, parameterName=NoiseTexture_Pan_U}
  n8 = ScalarParameter() {group=Noise, parameterName=NoiseTexture_Pan_V}
  n9 = TextureSampleParameter2D(UVs=Panner(Coordinate=Multiply(A=@n5, B=@n6), Speed=AppendVector(A=@n7, B=@n8)) {speedY=0.3}) {group=Noise, parameterName=NoiseTexture, texture=/Game/Packages/Fortress_Transylvania/Haunted_Forest/Swamps/Textures/2ndInfectedMapLinear2}
  n10 = TextureCoordinate()
  n11 = AppendVector(
   A=ScalarParameter() {defaultValue=1.5, group=Noise, parameterName=NoiseTexture_Tile_U 2}
   B=ScalarParameter() {defaultValue=0.5, group=Noise, parameterName=NoiseTexture_Tile_V 2}
  )
  n12 = ScalarParameter() {group=Noise, parameterName=NoiseTexture_Pan_U 2}
  n13 = ScalarParameter() {group=Noise, parameterName=NoiseTexture_Pan_V 2}
  n14 = TextureSampleParameter2D(UVs=Panner(Coordinate=Multiply(A=@n10, B=@n11), Speed=AppendVector(A=@n12, B=@n13)) {speedY=0.3}) {
   group=Noise
   parameterName=NoiseTexture 2
   samplerType=SAMPLERTYPE_LinearGrayscale
   texture=&/T_SnowFall_pick_basecolor
  }
  n15 = MaterialFunctionCall(UV Channel=LinearInterpolate(A=TextureCoordinate(), B=TextureCoordinate() {coordinateIndex=1}, Alpha=ScalarParameter() {parameterName=noise transition Use UV1})) {materialFunction=&LinearGradient}
  n16 = ScalarParameter() {defaultValue=1, parameterName="noise transition Use V Gradient "}
  n17 = Add(
   A=OneMinus(None=LinearInterpolate(A=@n15.UGradient, B=@n15.UGradient, Alpha=@n16))
   B=ScalarParameter() {defaultValue=-0.2, group=Noise, parameterName=noise Transition gradient Offset}
  )
  n18 = ScalarParameter() {defaultValue=3, group=Noise, parameterName=noise Transition gradient Power}
  n19 = Saturate(None=Multiply(A=Power(Base=@n17, Exp=@n18), B=ScalarParameter() {defaultValue=5, group=Noise, parameterName=noise Transition gradient Intencity}))
  n20 = Multiply(
   A=Power(Base=LinearInterpolate(A=@n9.R, B=@n14.R, Alpha=@n19), Exp=ScalarParameter() {defaultValue=1, group=Noise, parameterName=NoiseTexture_Power})
   B=ScalarParameter() {defaultValue=1, group=Noise, parameterName=NoiseTexture_Intensity}
  )
  n21 = ScalarParameter() {defaultValue=1, group=Color, parameterName=Secondary Color Intensity}
  n22 = DynamicParameter() {paramNames [PrimaryR,PrimaryG,PrimaryB,PrimaryA]}
  n23 = AppendVector(A=AppendVector(A=AppendVector(A=@n22.PrimaryR, B=@n22.PrimaryR), B=@n22.PrimaryB), B=@n22.PrimaryA)
  n24 = ScalarParameter() {defaultValue=1, group=Color, parameterName=Primary Color Intensity}
  n25 = CurveAtlasRowParameter(CurveTime=ComponentMask(None=TextureCoordinate()) {r=true}) {
   atlas=/Game/Athena/LinearColorCurves/TempAthenaCurveAtlas
   curve=/Game/Weapons/FORT_Melee/Pickaxe_Elastic/Fire/Curve_Pickaxe_Elastic_Fire_01
   defaultValue=14
  }
  n26 = MaterialFunctionCall(Float2=LinearInterpolate(A=Constant3Vector() {constant=<rgba 1 1 1 0>}, B=@n25, Alpha=ScalarParameter() {defaultValue=1, parameterName=Use Param})) {materialFunction=&BreakOutFloat2Components}
  n27 = Saturate(None=Multiply(A=@n26.R) {constB=10})
  n28 = ScalarParameter() {group=Color, parameterName=Use_clear_boundaries}
  n29 = LinearInterpolate(
   A=Desaturation(None=Multiply(A=@n2, B=@n21), Fraction=ScalarParameter() {defaultValue=0.2, group=Color, parameterName=Secondary Color Desaturation})
   B=Desaturation(None=Multiply(A=@n23, B=@n24), Fraction=ScalarParameter() {defaultValue=0.15, group=Color, parameterName=Primary Color Desaturation})
   Alpha=LinearInterpolate(A=LinearInterpolate(A=@n26.R, B=@n27, Alpha=@n28), B=Saturate(None=@n20), Alpha=ScalarParameter() {parameterName=use noise as color lerp})
  )
  n30 = DynamicParameter() {parameterIndex=2, paramNames [inner color R, inner color G, inner color B, inner color A]}
  n31 = MaterialFunctionCall(R=@n30.inner color R, G=@n30.inner color R, B=@n30.inner color R, A=@n30.inner color R) {materialFunction=&AppendMany}
  n32 = VertexColor()
  n33 = Multiply(
   A=LinearInterpolate(A=@n29, B=LinearInterpolate(A=@n31.RGB, B=@n29, Alpha=@n32.R), Alpha=ScalarParameter() {parameterName=use inner color})
   B=@n26.G
  )
  n34 = LinearInterpolate(
   A=Multiply(A=Desaturation(None=@n3, Fraction=@n4), B=@n20)
   B=LinearInterpolate(A=@n33, B=Multiply(A=@n33, B=@n20), Alpha=ScalarParameter() {defaultValue=1, parameterName=Multiply mask in color})
   Alpha=Clamp(None=TwoSidedSign())
  )
  n35 = ParticleColor()
  MaterialFunctionCall(Emissive=LinearInterpolate(A=@n34, B=Multiply(A=@n34, B=@n35.RGB), Alpha=ScalarParameter() {parameterName=Use_ParticleColor})) {materialFunction=&MF_FX_EmissiveAdjustments}
  n36 = CameraPositionWS()
  n37 = VectorParameter() {parameterName=Actor Pos Offset, channelNames {a=Use Local Space Offset, b=Z, g=Y, r=X}, defaultValue=<rgba 0 0 0 1>}
  n38 = Add(
   A=LinearInterpolate(A=ActorPositionWS(), B=ObjectPositionWS(), Alpha=ScalarParameter() {defaultValue=1, parameterName=Actor vs Object position})
   B=LinearInterpolate(A=@n37, B=Transform(None=@n37) {transformSourceType=TRANSFORMSOURCE_Instance}, Alpha=@n37)
  )
  n39 = DotProduct(
   A=Normalize(VectorInput=Subtract(A=@n36, B=@n38))
   B=Transform(None=VectorParameter() {parameterName=Object vector to comapre, defaultValue=<rgba 0 1 0 1>}.RGB) {transformSourceType=TRANSFORMSOURCE_Instance}
  )
  n40 = VectorParameter() {parameterName=Dot Options, channelNames {a=pre mul, b=add, g=mul, r=pow}, defaultValue=<rgba 1 1 0 4>}
  n41 = Multiply(A=OneMinus(None=ConstantBiasScale(None=@n39)), B=@n40.pre mul)
  n42 = Multiply(A=Power(Base=Clamp(None=@n41), Exp=@n40.pow) {constExponent=1}, B=@n40.mul)
  n43 = LinearInterpolate(
   A=ScalarParameter() {parameterName=opacity - look at}
   B=ScalarParameter() {defaultValue=1, parameterName=opacity - look away}
   Alpha=Clamp(None=Add(A=@n42, B=@n40.add) {constB=0})
  )
  n44 = ScalarParameter() {parameterName=Cam Fade length}
  n45 = ScalarParameter() {parameterName=Cam Fade offset}
  n46 = StaticBool() {value=true}
  n47 = ScalarParameter() {defaultValue=2, parameterName=AngleFade}
  n48 = MaterialFunctionCall("FadeSensitivity (optional)"=@n47) {materialFunction=&FadeBasedOnViewAngle}
  n49 = CameraVectorWS()
  n50 = VertexNormalWS()
  n51 = Clamp(None=Power(Base=DotProduct(A=@n49, B=@n50), Exp=@n47))
  n52 = ScalarParameter() {parameterName=FadeBackFaces}
  n53 = StaticSwitchParameter(
   True=Constant() {r=1}
   False=DepthFade(Opacity=LinearInterpolate(A=@n48.Result, B=@n51, Alpha=@n52), FadeDistance=ScalarParameter() {defaultValue=16, parameterName=FadeDistance})
  ) {parameterName=IsMasked/opaque}
  n54 = TextureSampleParameter2D() {parameterName=Opacity Texture, samplerType=SAMPLERTYPE_Grayscale, texture=&LowResBlurredNoise}
  n55 = Multiply(
   A=Multiply(A=@n35.A, B=@n53)
   B=LinearInterpolate(
    A=ScalarParameter() {defaultValue=0.2, parameterName=Opacity Texture Low}
    B=ScalarParameter() {defaultValue=1, parameterName=Opacity Texture High}
    Alpha=@n54.R
   ) {constA=0.2}
  )
  n56 = LinearInterpolate(A=TextureCoordinate(), B=TextureCoordinate() {coordinateIndex=1}, Alpha=ScalarParameter() {parameterName=Opacity gradient Use UV1})
  n57 = LinearInterpolate(
   B=Power(Base=ComponentMask(None=@n56) {g=true}, Exp=ScalarParameter() {defaultValue=1.5, parameterName=Opacity gradient power}) {constExponent=1.5}
   Alpha=ScalarParameter() {defaultValue=1, parameterName=Use Opecity Gradient}
  ) {constA=1}
  n58 = TextureCoordinate()
  n59 = AppendVector(A=ScalarParameter() {parameterName=Erode Speed x}, B=ScalarParameter() {parameterName=Erode Speed y})
  n60 = ScalarParameter() {parameterName=Erode tile X}
  n61 = ScalarParameter() {parameterName=Erode tile y}
  n62 = TextureSampleParameter2D(UVs=Multiply(A=Panner(Coordinate=@n58, Speed=@n59), B=AppendVector(A=@n60, B=@n61))) {parameterName=Erode Texture, samplerType=SAMPLERTYPE_LinearGrayscale, texture=/Game/Effects/Fort_Effects/Textures/T_DarkAndLight_Noise}
  n63 = DynamicParameter() {parameterIndex=3, defaultValue=<rgba 0 1 1 1>, paramNames [erode Amount, Fade in gradient amount, Param3, Param4]}
  n64 = Add(A=@n62.R, B=Multiply(A=Subtract(A=@n63.erode Amount) {constB=0.5}) {constB=2})
  n65 = ScalarParameter() {defaultValue=2, parameterName=erode contrast}
  n66 = LinearInterpolate(
   A=@n53
   B=Multiply(
    A=Multiply(A=@n55, B=@n57)
    B=MaterialFunctionCall(In=@n64, Contrast=@n65) {materialFunction=&CheapContrast}.Result
   )
   Alpha=ScalarParameter() {parameterName=Use_Particle alpha}
  )
  n67 = LinearInterpolate(A=TextureCoordinate(), B=TextureCoordinate() {coordinateIndex=1}, Alpha=ScalarParameter() {parameterName=Fade in gradient Use UV1})
  n68 = MaterialFunctionCall(
   In=Add(
    A=MaterialFunctionCall(UV Channel=@n67) {materialFunction=&LinearGradient}.VGradient
    B=@n63.Fade in gradient amount
   )
   Contrast=ScalarParameter() {defaultValue=2, parameterName=Fadin gradient contrast}
  ) {materialFunction=&CheapContrast}
  n69 = TextureCoordinate()
  n70 = Abs(None=Subtract(A=ComponentMask(None=@n69) {g=true}) {constB=0.5})
  n71 = Clamp(None=Subtract(A=OneMinus(None=@n70)) {constB=0.5})
  n72 = ScalarParameter() {parameterName=Use Opacity Band}
  Multiply(
   A=Multiply(
    A=Clamp(None=@n43)
    B=MaterialFunctionCall(Fade Length=@n44, Fade Offset=@n45, For Vertex Shader=@n46) {materialFunction=&CameraDepthFade}.Result
   )
   B=Multiply(A=Multiply(A=@n66, B=@n68.Result), B=LinearInterpolate(B=@n71, Alpha=@n72) {constA=1})
  )
  n73 = TextureCoordinate()
  n74 = AppendVector(
   A=ScalarParameter() {defaultValue=2, group=WPO, parameterName=WPOTexture_Tile_U}
   B=ScalarParameter() {defaultValue=0.5, group=WPO, parameterName=WPOTexture_Tile_V}
  )
  n75 = ScalarParameter() {group=WPO, parameterName=WPOTexture_Pan_U}
  n76 = ScalarParameter() {group=WPO, parameterName=WPOTexture_Pan_V}
  n77 = TextureSampleParameter2D(UVs=Panner(Coordinate=Multiply(A=@n73, B=@n74), Speed=AppendVector(A=@n75, B=@n76)) {speedY=0.3}) {group=WPO, parameterName=WPO_Texture, samplerType=SAMPLERTYPE_LinearColor, texture=&HighContrastNoise}
  n78 = Multiply(A=ScalarParameter() {defaultValue=1, group=WPO, parameterName=WPO_Intensity}, B=@n77.RGB)
  n79 = LinearInterpolate(B=@n71, Alpha=ScalarParameter() {defaultValue=1, parameterName=Use WPO Gradient}) {constA=1}
  n80 = VertexNormalWS()
  n81 = WorldPosition()
  n82 = VectorParameter() {parameterName=WPO PositionScale, channelNames {a=Master, b=Z Mul, g=Y Mul, r=X Mul}, defaultValue=<rgba 2 2 2 1>}
  n83 = Multiply(A=@n82, B=@n82)
  n84 = ComponentMask(None=TransformPosition(None=Divide(A=@n81.XYZ, B=@n83)) {transformSourceType=TRANSFORMPOSSOURCE_World, transformType=TRANSFORMPOSSOURCE_Instance}) {g=true, r=true}
  n85 = AppendVector(
   A=ScalarParameter() {defaultValue=2, group=WPO, parameterName=WPOTexture_Tile_U}
   B=ScalarParameter() {defaultValue=0.5, group=WPO, parameterName=WPOTexture_Tile_V}
  )
  n86 = ScalarParameter() {group=WPO, parameterName=WPOTexture_Pan_U}
  n87 = ScalarParameter() {group=WPO, parameterName=WPOTexture_Pan_V}
  n88 = TextureSampleParameter2D(UVs=Panner(Coordinate=Divide(A=@n84, B=@n85), Speed=AppendVector(A=@n86, B=@n87))) {group=WPO, parameterName=WPO_Texture, samplerType=SAMPLERTYPE_LinearColor, texture=&HighContrastNoise}
  StaticSwitchParameter(
   True=Transform(None=Multiply(A=@n78, B=@n79))
   False=Multiply(A=ScalarParameter() {defaultValue=1, group=WPO, parameterName=WPO_Intensity}, B=Multiply(A=@n80, B=@n88.RGB))
  ) {defaultValue=true, parameterName=Use Old WPO Method}
 ]
}
