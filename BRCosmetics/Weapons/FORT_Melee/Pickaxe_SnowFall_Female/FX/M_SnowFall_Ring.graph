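# Translucent unlit Niagara mesh-particle material for the SnowFall pickaxe ring effect.
# Emissive blends a grayscale inner value from the dynamic parameter's red channel with particle color by vertex red.
# Opacity multiplies a contrast-adjusted gradient and generated band on noise-distorted UVs, with particle alpha and depth fade; mobile substitutes a scalar.
# A noise-distorted, panning sine builds a vertex-normal wobble offset, signed by vertex red.
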Material /BRCosmetics/Weapons/FORT_Melee/Pickaxe_SnowFall_Female/FX/M_SnowFall_Ring
graph {
 properties {
  bUsedWithNiagaraMeshParticles=true
  blendMode=BLEND_Translucent
  pixelDepthOffsetMode=PDOM_Legacy
  previewMesh=&/SM_SnowFallRing
  shadingModel=MSM_Unlit
  usedShadingModels=Unlit
 }
 nodes [
  ComponentMask(None=TextureCoordinate()) {g=true}
  n1 = DynamicParameter() {defaultValue=<rgba 0 1 1 1>, paramNames [Particle color inner R, Particle color inner G, Particle color inner B, Particle color inner A]}
  n2 = ParticleColor()
  MaterialFunctionCall(Emissive=LinearInterpolate(
    A=MaterialFunctionCall(R=@n1.Particle color inner R, G=@n1.Particle color inner R, B=@n1.Particle color inner R, A=@n1.Particle color inner R) {materialFunction=&AppendMany}.RGB
    B=@n2.RGB
    Alpha=VertexColor().R
   )) {materialFunction=&MF_FX_EmissiveAdjustments}
  n3 = VertexColor()
  n4 = VertexNormalWS()
  n5 = ScalarParameter() {defaultValue=1, parameterName=wobble Tile distort X}
  n6 = ScalarParameter() {defaultValue=1, parameterName=wobble Tile Distort Y}
  n7 = Panner(
   Coordinate=Multiply(A=TextureCoordinate(), B=AppendVector(A=@n5, B=@n6))
   Speed=AppendVector(
    A=ScalarParameter() {defaultValue=0.1, parameterName=wobble Speed distort X}
    B=ScalarParameter() {defaultValue=0.1, parameterName=wobble Speed Distort Y}
   )
  )
  n8 = Subtract(A=ComponentMask(None=TextureSampleParameter2D(UVs=@n7) {parameterName=Wooble Noise, samplerType=SAMPLERTYPE_LinearColor, texture=&HighContrastNoise}.RGB) {g=true, r=true}) {constB=0.5}
  n9 = ScalarParameter() {defaultValue=0.1, parameterName=Wobble distort Str}
  n10 = Multiply(
   A=Add(A=TextureCoordinate(), B=Multiply(A=@n8, B=@n9))
   B=AppendVector(A=ScalarParameter() {defaultValue=1, parameterName=wobble Tile X}, B=ScalarParameter() {defaultValue=1, parameterName=wobble Tile Y})
  )
  n11 = ScalarParameter() {defaultValue=1, parameterName=wobble Speed}
  n12 = Sine(None=MaterialFunctionCall(UV Channel=Panner(Coordinate=@n10, Speed=@n11)) {materialFunction=&LinearGradient}.VGradient)
  Multiply(
   A=Multiply(A=ConstantBiasScale(None=@n3.R) {bias=-0.5, scale=2}, B=Multiply(A=@n4, B=@n12))
   B=ScalarParameter() {defaultValue=1, parameterName=WPO Str}
  )
  ScalarParameter() {defaultValue=0.5, parameterName=Specular}
  ScalarParameter() {parameterName=Roughness}
  ScalarParameter() {parameterName=Metallic}
  n13 = ScalarParameter() {defaultValue=-0.048, parameterName=Distort Low}
  n14 = ScalarParameter() {defaultValue=0.5, parameterName=Distort High}
  n15 = OneMinus(None=MaterialFunctionCall(
    In=MaterialFunctionCall() {materialFunction=&LinearGradient}.VGradient
    Contrast=Constant() {r=3}
   ) {materialFunction=&CheapContrast}.Result)
  n16 = Constant()
  n17 = TextureCoordinate()
  n18 = AppendVector(A=ScalarParameter() {defaultValue=0.5, parameterName=Distort tile X}, B=ScalarParameter() {defaultValue=0.25, parameterName=Distort tile y})
  n19 = ScalarParameter() {parameterName=Distort Speed X}
  n20 = ScalarParameter() {parameterName=Distort Speed Y}
  n21 = TextureSample(UVs=Panner(Coordinate=Multiply(A=@n17, B=@n18), Speed=AppendVector(A=@n19, B=@n20)) {speedY=-1}) {samplerType=SAMPLERTYPE_LinearColor, texture=&HighContrastNoise}
  n22 = Add(A=TextureCoordinate(), B=Multiply(A=LinearInterpolate(A=@n13, B=@n14, Alpha=@n15), B=AppendVector(A=@n16, B=@n21.R)) {constA=0.25})
  n23 = VectorParameter() {parameterName=opacity mas options, channelNames {a=Contrast, b=add, g=multiply, r=power}, defaultValue=<rgba 1 1 0 0>}
  n24 = Multiply(
   A=Power(
    Base=MaterialFunctionCall(UV Channel=@n22) {materialFunction=&LinearGradient}.VGradient
    Exp=@n23.power
   )
   B=@n23.multiply
  )
  n25 = ScalarParameter() {defaultValue=0.4, parameterName=Band width low}
  n26 = ScalarParameter() {defaultValue=0.4, parameterName=Band width high}
  n27 = Multiply(
   A=MaterialFunctionCall(In=Add(A=@n24, B=@n23.add), Contrast=@n23.Contrast) {materialFunction=&CheapContrast}.Result
   B=MaterialFunctionCall!(
    Width=LinearInterpolate(A=@n25, B=@n26, Alpha=@n15)
    Sharpness=Constant() {r=0.95}
    Offset
    Direction Switch
    Compare
    Input Coordinates=Saturate(None=@n22)
   ) {materialFunction=/Engine/Functions/Engine_MaterialFunctions03/Procedurals/GeneratedBand}.Result
  )
  ShadingPathSwitch(
   Default=DepthFade(Opacity=Multiply(A=@n2.A, B=@n27), FadeDistance=ScalarParameter() {parameterName=FadeDistance}) {fadeDistanceDefault=15}
   Mobile=ScalarParameter() {defaultValue=0.5, group=Opacity, parameterName=Mobile Opacity, sortPriority=10}
  )
 ]
}
