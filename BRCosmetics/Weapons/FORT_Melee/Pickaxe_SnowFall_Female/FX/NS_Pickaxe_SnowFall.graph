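# Niagara system for the SnowFall pickaxe cosmetic, layering an icy flame, glow and light, clouds, cold smoke and snowflakes.
# Mesh and sprite emitters build the looping body and wisps with curl noise, vortex and point attraction.
# Many layers pass colors, often from User Primary and Secondary Color, through time-of-day selection with hue shifts.
# A system Timeline drives flame stretch, glow fade, burst speed and some wisp positions.
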NiagaraSystem /BRCosmetics/Weapons/FORT_Melee/Pickaxe_SnowFall_Female/FX/NS_Pickaxe_SnowFall
dependencies {
 usedDataInterfaces [
  &NiagaraDataInterfaceCurve
  &NiagaraDataInterfaceMeshRendererInfo
  &NiagaraDataInterfaceVectorCurve
  &NiagaraDataInterfaceArrayColor
 ]
 usedDynamicInputs [&SetBoolByFloatComparison]
 usedModules [&SystemState, &Timeline]
}
emitters {
 Clouds {
  emitterUpdateScript [
   &EmitterState(Life Cycle Mode=::NewEnumerator1, Loop Behavior=::NewEnumerator1, Loop Duration Mode=::NewEnumerator1, UseLoopDelay=false)
   &SpawnRate(Use Spawn Probability=false)
   &SpawnBurst_Instantaneous(Use Spawn Probability=false, UseLoopCountLimit=false)
  ]
  particleSpawnScript [
   &InitializeParticle(
    Lifetime=3
    Mesh Scale Max=<1 1 2>
    Mesh Scale Mode=::NewEnumerator2
    Position Offset=<5 0 0>
    Sprite Size Mode=::NewEnumerator3
    Uniform Sprite Size=1
    UsePositionOffset=true
    Write Material Random=true
   )
   &InitialMeshOrientation(
    Mesh Orientation Mode=::NewEnumerator7
    Rotation=&RandomRangeVector(Minimum=<0 0.25 0.15>, Maximum=<0 0.25 0.05>)
    Use Rotation Vector=true
   )
   &InitialMeshOrientation(
    Mesh Orientation Mode=::NewEnumerator7
    Rotation=&RandomRangeVector(Minimum=<0 0 0>, Maximum=<1 0 0>)
    Use Rotation Vector=true
   ) {module=InitialMeshOrientation001}
   &InitialMeshOrientation(Mesh Orientation Mode=::NewEnumerator6, Use Rotation Vector=true) {enabled=false, module=InitialMeshOrientation002}
   &ShapeLocation(
    Cylinder Distribution=::NewEnumerator2
    Cylinder Height=8
    Cylinder Height Midpoint=1.25
    Cylinder Radius=1
    MaskSpawnGroup=false
    Rotation Mode=::NewEnumerator1
    Shape Primitive=::NewEnumerator1
    U Position=&RandomRangeFloat(Minimum=0, Maximum=1)
    UseOwnerScale=true
    V Position=&RandomRangeFloat(Minimum=0, Maximum=1)
   )
  ]
  particleUpdateScript [
   &SolveForcesAndVelocity(Clamp Velocity=false, Limit Acceleration=false) {enabled=false}
   &ParticleState()
   &UpdateMeshOrientation(
    Rotation Coordinate Space=::NewEnumerator2
    Rotation Rate=&FloatFromCurve(FloatCurve=DataInterface() {propertyValues {Curve {keys [<at 0.65 1 RCIM_Cubic>,<at 1 0 RCIM_Cubic>]}}}, Scale Curve=0.075)
   )
   &UpdateMeshOrientation(
    Rotation Coordinate Space=::NewEnumerator2
    Rotation Rate=&FloatFromCurve(
     FloatCurve=DataInterface() {propertyValues {Curve {keys [<at 0.65 0 RCIM_Cubic>,<at 1 1 RCIM_Cubic>]}}}
     Scale Curve=&RandomRangeFloat(Minimum=0.1, Maximum=0.2)
    )
    Rotation Vector=&RandomVector(Vector Scale=1)
   ) {module=UpdateMeshOrientation001}
   &ScaleColor(
    Scale Alpha=&FloatFromCurve(FloatCurve=DataInterface() {propertyValues {Curve {keys [
         {interpMode=RCIM_Cubic, tangentMode=RCTM_User, time=0, value=0}
         {interpMode=RCIM_Cubic, tangentMode=RCTM_User, time=0.5, value=1}
         {tangentMode=RCTM_User, time=1, value=0}
        ]}}})
    ScaleA=true
    ScaleRGB=true
   )
   /SetVariables(
    Particles.DynamicMaterialParameter:Vector4f=&MakeVector4_FromLinearColor(LinearColor=<rgba 0.2423 0.5089 0.7011 1>)
    Particles.DynamicMaterialParameter2:Vector4f=&MakeVector4_FromLinearColor(LinearColor=<rgba 0.15 0.5292 1 1>)
   ) {enabled=false, isSetParameters=true}
   &ScaleMeshSize(Scale Factor=&VectorFromCurve(VectorCurve=DataInterface() {propertyValues {
       XCurve {keys [<at 0 0.25 RCIM_Cubic>,<at 1 1 RCIM_Cubic>]}
       YCurve {keys [<at 0 0.25 RCIM_Cubic>,<at 1 1 RCIM_Cubic>]}
       ZCurve {keys [{arriveTangent=2.7756, interpMode=RCIM_Cubic, leaveTangent=2.7756, tangentMode=RCTM_User, time=0, value=1}, <at 1 0.25 RCIM_Cubic>]}
      }}))
   /SetVariables(
    Particles.DynamicMaterialParameter:Vector4f=&MakeVector4_FromLinearColor(LinearColor=<rgba 0.0886 0.3126 0.5906 1>)
    Particles.DynamicMaterialParameter1:Vector4f=&MakeVector4_FromLinearColor(LinearColor=<rgba 0.3755 0.822 0.9387 1>)
   ) {isSetParameters=true}
   &self:DynParam_ColorPicker(
    Primary Color:LinearColor=&FN_Select_Color_from_Time_Of_Day(
     Morning=&HueShiftLinearColor(Hue=0, Saturation=1, Value=0.5, Linear Color=<rgba 0.0886 0.3126 0.5906 1>)
     Day=<rgba 0.0886 0.3126 0.5906 1>
     Evening=&HueShiftLinearColor(Hue=0, Saturation=1, Value=0.25, Linear Color=<rgba 0.0886 0.3126 0.5906 1>)
     Night=&HueShiftLinearColor(Hue=0, Saturation=0.25, Value=0.85, Linear Color=<rgba 0.0886 0.3126 0.5906 1>)
    )
    Secondary Color:LinearColor=&FN_Select_Color_from_Time_Of_Day(
     Morning=&HueShiftLinearColor(Hue=0, Saturation=1, Value=0.5, Linear Color=<rgba 0.3755 0.822 0.9387 1>)
     Day=<rgba 0.3755 0.822 0.9387 1>
     Evening=&HueShiftLinearColor(Hue=0, Saturation=1, Value=1, Linear Color=<rgba 0.3755 0.822 0.9387 1>)
     Night=&HueShiftLinearColor(Hue=0.15, Saturation=0.25, Value=1, Linear Color=<rgba 0.3755 0.822 0.9387 1>)
    )
   ) {module=DynParam_ColorPicker}
   /SetVariables(
    Particles.DynamicMaterialParameter2:Vector4f=&MakeVector4_FromLinearColor(LinearColor=<rgba 0.06 0.1314 0.15 1>)
    Particles.DynamicMaterialParameter3:Vector4f=&MakeVector4(
     X=&FloatFromCurve(FloatCurve=DataInterface() {propertyValues {Curve {keys [
          {interpMode=RCIM_Cubic, leaveTangentWeight=0.0333, tangentMode=RCTM_User, time=0, value=0}
          {arriveTangent=1.6667, arriveTangentWeight=0.0581, interpMode=RCIM_Cubic, leaveTangent=1.6667, leaveTangentWeight=0.3488, time=0.25, value=1}
          {arriveTangent=-1.3333, arriveTangentWeight=0.299, interpMode=RCIM_Cubic, leaveTangent=-1.3333, leaveTangentWeight=0.1495, time=0.6, value=1}
          {arriveTangentWeight=0.1, interpMode=RCIM_Cubic, tangentMode=RCTM_User, time=1, value=0}
         ]}}})
     Y=0
     Z=1
     W=0
    )
   ) {isSetParameters=true}
  ]
  rendererClasses [&NiagaraSpriteRendererProperties, &NiagaraMeshRendererProperties]
  renderers [
   {rendererClass=&NiagaraSpriteRendererProperties}
   {rendererClass=&NiagaraMeshRendererProperties, rendererIndex=1}
  ]
 }
 Cold_Smoke {
  emitterUpdateScript [
   &EmitterState(Inactive Response=::NewEnumerator1, Life Cycle Mode=::NewEnumerator1, Loop Behavior=::NewEnumerator1, Loop Duration=1, UseLoopDelay=true)
   &SpawnBurst_Instantaneous(
    Spawn Count=&RandomRangeInt(Minimum=3, Maximum=5)
    Use Spawn Probability=false
    UseLoopCountLimit=false
   )
  ]
  particleSpawnScript [
   /SetVariables(
    Particles.Random Int:NiagaraInt32=&RandomRangeInt(Minimum=0, Maximum=32)
    Particles.speed life:NiagaraFloat=&FloatFromCurve(
     FloatCurve=DataInterface() {propertyValues {Curve {keys [<at 0 2>,<at 1 0.15>]}}}
     CurveIndex=&Divide_Float(A=&VectorLength(Vector=$Engine.Owner.Velocity), B=250)
    )
   ) {isSetParameters=true}
   &InitializeParticle(
    Color=&MultiplyLinearColor(Linear Color=$User.Primary Color, Scale Factor=<1 1 1 0.1>)
    Lifetime=0.75
    Sprite Rotation Mode=::NewEnumerator2
    Sprite Size Mode=::NewEnumerator1
    Sprite UV Mode=::NewEnumerator3
    Uniform Sprite Size Max=90
    Uniform Sprite Size Min=60
    UsePositionOffset=false
    Write Material Random=true
   )
   &Color(Color=&FN_Select_Color_from_Time_Of_Day(
     Morning=&HueShiftLinearColor(Hue=0, Saturation=1, Value=0.5, Linear Color=$Particles.Initial.Color)
     Day=$Particles.Initial.Color
     Evening=&HueShiftLinearColor(Hue=0, Saturation=1, Value=0.25, Linear Color=$Particles.Initial.Color)
     Night=&HueShiftLinearColor(Hue=0, Saturation=0.25, Value=0.9, Linear Color=$Particles.Initial.Color)
    ))
   &CameraOffset(Camera Offset Amount=&RandomRangeFloat(Minimum=20, Maximum=20))
  ]
  particleUpdateScript [
   &ParticleState()
   &CurlNoiseForce(Noise Frequency=25, Noise Strength=5, Pan Noise=true, Pan Noise Field=<0 0 0.1>)
   &Drag(Drag=&RandomRangeFloat(Minimum=10, Maximum=8), Use Linear Drag=true, Use Rotational Drag=false)
   &ScaleSpriteSize(Scale Factor=&Vector2DFromFloat(Value=&FloatFromCurve(FloatCurve=DataInterface() {propertyValues {Curve {keys [
          {arriveTangent=0.1544, interpMode=RCIM_Cubic, leaveTangentWeight=0.3412, time=0, value=0}
          {arriveTangent=1.2455, interpMode=RCIM_Cubic, leaveTangent=1.2455, time=0.25, value=1.0173}
          {
           arriveTangent=-0.6558
           arriveTangentWeight=0.2975
           interpMode=RCIM_Cubic
           leaveTangent=-0.6558
           leaveTangentWeight=0.1523
           tangentMode=RCTM_User
           time=0.7001
           value=0.872
          }
          {interpMode=RCIM_Cubic, leaveTangent=-0.4051, time=1.0011, value=0.75}
         ]}}}))) {enabled=false}
   &ScaleColor(
    Scale Alpha=&FloatFromCurve(FloatCurve=DataInterface() {propertyValues {Curve {keys [
         {arriveTangent=6.8671, interpMode=RCIM_Cubic, leaveTangent=6.8671, tangentMode=RCTM_User, time=0, value=0}
         {arriveTangentWeight=0.7806, interpMode=RCIM_Cubic, leaveTangentWeight=0.8245, tangentMode=RCTM_Break, time=0.5, value=1}
         {interpMode=RCIM_Cubic, leaveTangent=-5.8107, time=1, value=0}
        ]}}})
    ScaleA=true
    ScaleRGB=true
   )
   &GravityForce(Gravity=<0 0 -50>) {enabled=false}
   &SolveForcesAndVelocity(Clamp Velocity=false, Limit Acceleration=false)
   &SubUVAnimation(SubUV Lookup Index:NiagaraFloat=$Engine.RealTime, SubUV Loop Count=0.25)
   &ScaleSpriteSize(
    Uniform Curve Scale=0.75
    Uniform Curve Sprite Scale=DataInterface() {propertyValues {Curve {keys [
        {arriveTangent=3.3327, interpMode=RCIM_Cubic, leaveTangent=3.3327, tangentMode=RCTM_User, time=0, value=0}
        {arriveTangent=0.1777, interpMode=RCIM_Cubic, leaveTangent=0.1777, tangentMode=RCTM_User, time=1, value=1}
       ]}}}
   ) {module=ScaleSpriteSize001}
   &SpriteRotationRate(Rotation Rate=&RandomRangeFloat(Minimum=-10, Maximum=10))
  ]
  rendererClasses [&NiagaraSpriteRendererProperties]
  renderers [{rendererClass=&NiagaraSpriteRendererProperties}]
 }
 Decal_Glow_Wrist {
  emitterUpdateScript [
   &EmitterState()
   &SpawnBurst_Instantaneous(Age:NiagaraFloat=$Emitter.Age, Use Spawn Probability=false, UseLoopCountLimit=false)
  ]
  particleSpawnScript [
   &InitializeParticle(
    Color=&MultiplyLinearColor(
     Linear Color=$User.Primary Color
     Scale Factor=&MakeFloat4FromVectorAndFloat("Vector (RGB)"=<1 1 1>, "Float (Alpha)"=0.35)
    )
    Lifetime=5
    Mesh Scale=<0.25 0.25 0.35>
    Mesh Scale Mode=::NewEnumerator4
    Sprite Rotation Mode=::NewEnumerator1
    Sprite Size Mode=::NewEnumerator3
    Uniform Sprite Size=40.3051
    UsePositionOffset=false
    Write Material Random=true
   ) {module=InitializeParticle001}
   &SystemLocation(MaskSpawnGroup=false, Offset=<5 0 0>)
   &CameraOffset(Camera Offset Amount=25)
   &InitialMeshOrientation(Mesh Orientation Mode=::NewEnumerator7, Rotation=<0 0.25 0>, Use Rotation Vector=true)
  ]
  particleUpdateScript [
   &ParticleState(
    Kill Particles When Lifetime Has Elapsed=false
    Let Infinitely Lived Particles Die When Emitter Deactivates=true
    Loop Particles Lifetime=true
   )
   &DynamicMaterialParameters(
    Index 0 Param 1=0.25
    Index 0 Param 2=1
    Module.Param0WriteEnabled=true
    Module.Param1WriteEnabled=true
    Module.Param2WriteEnabled=false
    Module.Param3WriteEnabled=false
   ) {module=DynamicMaterialParameters001}
   &ScaleColor(Scale RGB=&VectorFromFloat(Value=2), ScaleA=false, ScaleRGB=true)
   &Color(Color=&FN_Select_Color_from_Time_Of_Day(
     Morning=&HueShiftLinearColor(Hue=0, Saturation=1, Value=1, Linear Color=$Particles.Initial.Color)
     Day=&MultiplyLinearColor(
      Linear Color=$Particles.Initial.Color
      Scale Factor=&MakeFloat4FromVectorAndFloat("Vector (RGB)"=&VectorFromFloat(Value=2), "Float (Alpha)"=1)
     )
     Evening=&HueShiftLinearColor(Hue=0, Saturation=1, Value=0.5, Linear Color=$Particles.Initial.Color)
     Night=&HueShiftLinearColor(Hue=0, Saturation=0.6, Value=1.6, Linear Color=$Particles.Initial.Color)
    ))
  ]
  rendererClasses [&NiagaraSpriteRendererProperties]
  renderers [{rendererClass=&NiagaraSpriteRendererProperties}]
 }
 Flame_MAIN {
  emitterUpdateScript [
   &EmitterState()
   &SpawnBurst_Instantaneous(Age:NiagaraFloat=$Emitter.Age, Use Spawn Probability=false, UseLoopCountLimit=false)
  ]
  particleSpawnScript [
   &InitializeParticle(
    Color=$User.Primary Color
    Mesh Scale=&Multiply_Vector(A=<3.5 1.25 1.25>, B=&VectorFromFloat(Value=0.7273))
    Mesh Scale Mode=::NewEnumerator4
    UsePositionOffset=false
    Write Material Random=true
   ) {module=InitializeParticle001}
   &SystemLocation(MaskSpawnGroup=false, Offset=<-5 0 0>)
  ]
  particleUpdateScript [
   &ParticleState(Kill Particles When Lifetime Has Elapsed=false, Let Infinitely Lived Particles Die When Emitter Deactivates=true)
   &CameraOffset(Camera Offset Amount=15)
   &UpdateMeshOrientation(Rotation Coordinate Space=::NewEnumerator0, Rotation Rate=0.025) {module=UpdateMeshOrientation001}
   &self:DynParam_ColorPicker(
    Primary Color:LinearColor=&FN_Select_Color_from_Time_Of_Day(
     Morning=&HueShiftLinearColor(Hue=0, Saturation=1, Value=0.5, Linear Color=$User.Primary Color)
     Day=$User.Primary Color
     Evening=&HueShiftLinearColor(Hue=0, Saturation=1, Value=0.25, Linear Color=$User.Primary Color)
     Night=&HueShiftLinearColor(Hue=0, Saturation=0.6, Value=0.5, Linear Color=$User.Primary Color)
    )
    Secondary Color:LinearColor=&FN_Select_Color_from_Time_Of_Day(
     Morning=&HueShiftLinearColor(Hue=0, Saturation=1, Value=0.5, Linear Color=$User.Secondary Color)
     Day=$User.Secondary Color
     Evening=&HueShiftLinearColor(Hue=0, Saturation=1, Value=1, Linear Color=$User.Secondary Color)
     Night=&HueShiftLinearColor(Hue=0.15, Saturation=0.25, Value=1, Linear Color=$User.Secondary Color)
    )
   ) {module=DynParam_ColorPicker}
   &DynamicMaterialParameters(
    Index 3 Param 2=&FloatFromCurve(FloatCurve=DataInterface() {propertyValues {Curve {keys [<at 0 -1>,<at 1 1>]}}}, CurveIndex=$System.Timeline.NormalizedTime)
    Index 3 Param 3=1
    Module.Param0WriteEnabled3=true
    Module.Param1WriteEnabled3=true
    Module.Param2WriteEnabled3=true
    Module.Param3WriteEnabled3=true
    Write Parameter Index 0=false
    Write Parameter Index 3=true
   )
   &ScaleMeshSize(Scale Factor=&MakeVector(
     X=&FloatFromCurve(
      FloatCurve=DataInterface() {propertyValues {Curve {keys [
          {
           arriveTangentWeight=0.7973
           interpMode=RCIM_Cubic
           leaveTangentWeight=0.7973
           tangentMode=RCTM_User
           tangentWeightMode=RCTWM_WeightedBoth
           time=0
           value=0.75
          }
          {
           arriveTangentWeight=0.6983
           interpMode=RCIM_Cubic
           leaveTangentWeight=0.6983
           tangentMode=RCTM_User
           tangentWeightMode=RCTWM_WeightedBoth
           time=1
           value=1
          }
         ]}}}
      CurveIndex=$System.Timeline.NormalizedTime
     )
     Y=1
     Z=1
    ))
  ]
  rendererClasses [&NiagaraMeshRendererProperties]
  renderers [{rendererClass=&NiagaraMeshRendererProperties, rendererIndex=0}]
 }
 Fountain002 {
  emitterUpdateScript [
   &EmitterState(Life Cycle Mode=::NewEnumerator1, Loop Behavior=::NewEnumerator1, Loop Duration=0.2, UseLoopDelay=false)
   &SpawnBurst_Instantaneous(Use Spawn Probability=false, UseLoopCountLimit=false)
   &SpawnRate(SpawnRate=30, Use Spawn Probability=false)
  ]
  particleSpawnScript [
   &InitializeParticle(
    Color=<rgba 0.5434 0.7934 1 1>
    Lifetime Max=0.4
    Lifetime Min=0.3
    Lifetime Mode=::NewEnumerator1
    Mass Mode=::NewEnumerator3
    Mesh Scale=&RandomRangeVector(Minimum=<0.1 0.1 0.1>, Maximum=<0.1 0.1 0.5>)
    Mesh Scale Mode=::NewEnumerator4
    Sprite Rotation Mode=::NewEnumerator2
    Sprite Size Mode=::NewEnumerator1
    Uniform Sprite Size Max=12
    Uniform Sprite Size Min=6
    UsePositionOffset=false
    Write Material Random=true
   )
   &InitialMeshOrientation(
    Mesh Axis To Orient=<0 0 1>
    Mesh Orientation Mode=::NewEnumerator2
    Rotation=&RandomRangeVector(Minimum=<0 0 0>, Maximum=<0 0 1>)
    Use Rotation Vector=true
   )
  ]
  particleUpdateScript [
   &ParticleState()
   &ScaleColor(
    Scale Alpha=&FloatFromCurve(FloatCurve=DataInterface() {propertyValues {Curve {keys [
         <at -0.0046 1.0113 RCIM_Cubic>
         {arriveTangent=-2.7382, interpMode=RCIM_Cubic, leaveTangent=-2.7382, tangentMode=RCTM_User, time=1.0011, value=0.002}
        ]}}})
    ScaleA=true
    ScaleRGB=true
   )
   /SetVariables(Particles.Position=&Lerp_Vector(A=<0 0 0>, B=<-68 0 0>, Alpha=$Particles.NormalizedAge)) {isSetParameters=true}
   &DynamicMaterialParameters(
    Index 3 Param 3=&RandomRangeFloat(Minimum=0, Maximum=0.5)
    Index 3 Param 4=&RandomRangeFloat(Minimum=0, Maximum=0.5)
    Module.Param0WriteEnabled3=false
    Module.Param1WriteEnabled3=false
    Module.Param2WriteEnabled3=true
    Module.Param3WriteEnabled3=true
    Write Parameter Index 0=false
    Write Parameter Index 3=true
   )
   &UpdateMeshOrientation(Rotation Rate=&RandomRangeFloat(Minimum=-5, Maximum=-6), Rotation Vector=<0 0 1>)
   &ScaleMeshSize(Scale Factor=&VectorFromFloat(Value=&FloatFromCurve(FloatCurve=DataInterface() {propertyValues {Curve {keys [<at 0 0>,<at 0.3 1.125>,<at 0.5 1>,<at 1 0>]}}})))
  ]
  rendererClasses [&NiagaraMeshRendererProperties]
  renderers [{rendererClass=&NiagaraMeshRendererProperties, rendererIndex=0}]
 }
 LightGlow {
  emitterUpdateScript [
   &EmitterState()
   &SpawnRate(SpawnRate=5, Use Spawn Probability=false) {enabled=false}
   &SpawnBurst_Instantaneous(Age:NiagaraFloat=$Emitter.Age, Use Spawn Probability=false, UseLoopCountLimit=false)
  ]
  particleSpawnScript [
   &InitializeParticle(
    Color=&MultiplyLinearColor(Linear Color=$User.Primary Color, Scale Factor=<1 1 1 0.015>)
    Lifetime=5
    Sprite Size=<100 260>
    Sprite Size Mode=::NewEnumerator4
    UsePositionOffset=false
    Write Material Random=true
   ) {module=InitializeParticle001}
   &DynamicMaterialParameters(
    Index 0 Param 1=1
    Module.Param0WriteEnabled=true
    Module.Param1WriteEnabled=true
    Module.Param2WriteEnabled=true
    Module.Param3WriteEnabled=true
   )
   &SystemLocation(MaskSpawnGroup=false)
   &AddVelocity(Velocity=<1 0 0>) {enabled=false}
   /SetVariables(Particles.Velocity=<1 0 0>) {isSetParameters=true}
  ]
  particleUpdateScript [
   &CameraOffset(Camera Offset Amount=15)
   &ParticleState(
    Kill Particles When Lifetime Has Elapsed=false
    Let Infinitely Lived Particles Die When Emitter Deactivates=true
    Loop Particles Lifetime=true
   )
   &ScaleColor(
    Scale Alpha=&FloatFromCurve(
     FloatCurve=DataInterface() {propertyValues {Curve {keys [
         {interpMode=RCIM_Cubic, tangentMode=RCTM_User, time=0, value=0}
         {interpMode=RCIM_Cubic, tangentMode=RCTM_User, time=0.25, value=1}
         <at 1 0 RCIM_Cubic>
        ]}}}
     Scale Curve=0.015
    )
    Scale RGB=&VectorFromFloat(Value=1)
    ScaleA=true
    ScaleRGB=true
   ) {enabled=false}
   &FN_Light_v2(Light Exponent=3, Light Radius=25, LightEnabled=true, Position Offset:Vector3f=<0 5 -10>, Scale Color=75)
   &SolveForcesAndVelocity(Clamp Velocity=false, Limit Acceleration=false) {enabled=false}
   &ScaleSpriteSize(Uniform Curve Sprite Scale=DataInterface() {propertyValues {Curve {keys [
        {interpMode=RCIM_Cubic, tangentMode=RCTM_User, time=0, value=0.9}
        {interpMode=RCIM_Cubic, tangentMode=RCTM_User, time=0.5, value=1}
        {tangentMode=RCTM_User, time=1, value=0.9}
       ]}}})
   &Color(Color=&MultiplyLinearColor(
     Linear Color=&FN_Select_Color_from_Time_Of_Day(
      Morning=&HueShiftLinearColor(Hue=0, Saturation=1, Value=0.5, Linear Color=$Particles.Initial.Color)
      Day=$Particles.Initial.Color
      Evening=&HueShiftLinearColor(Hue=0, Saturation=1, Value=0.25, Linear Color=$Particles.Initial.Color)
      Night=&HueShiftLinearColor(Hue=0, Saturation=0.6, Value=0.9, Linear Color=$Particles.Initial.Color)
     )
     Scale Factor=&MakeVector4(
      X=1
      Y=1
      Z=1
      W=&FloatFromCurve(
       FloatCurve=DataInterface() {propertyValues {Curve {keys [{tangentMode=RCTM_User, time=0.75, value=0}, {tangentMode=RCTM_User, time=1, value=1}]}}}
       CurveIndex=$System.Timeline.NormalizedTime
      )
     )
    ))
  ]
  rendererClasses [&NiagaraSpriteRendererProperties, &NiagaraLightRendererProperties]
  renderers [
   {rendererClass=&NiagaraSpriteRendererProperties}
   {rendererClass=&NiagaraLightRendererProperties, rendererIndex=1}
  ]
 }
 PolkaDots001 {
  emitterUpdateScript [
   &EmitterState()
   &SpawnRate(SpawnRate=0.5, Use Spawn Probability=false) {enabled=false}
   &SpawnBurst_Instantaneous(Use Spawn Probability=false, UseLoopCountLimit=false) {enabled=false}
  ]
  particleSpawnScript [
   &InitializeParticle(
    Color=&SelectLinearColorFromArray(
     Array Sampling Mode=::NewEnumerator0
     Direct Array Index=?
     Color Selection Array=DataInterface() {propertyValues {ColorData [<rgba 0.3755 0.822 0.9387 1>, <rgba 0.6869 0.2137 0.855 1>, <rgba 0.5 0.5 0.5 1>]}}
     Random Seed=0
    )
    Lifetime Max=3
    Lifetime Min=2
    Lifetime Mode=::NewEnumerator1
    Mass Max=0.0025
    Mass Min=0.01
    Mass Mode=::NewEnumerator3
    Sprite Rotation Mode=::NewEnumerator2
    Sprite Size Mode=::NewEnumerator3
    Uniform Sprite Size=&Multiply_Float(A=&RandomRangeFloat(Minimum=0.75, Maximum=1.5), B=0.5)
    UsePositionOffset=true
    Write Material Random=true
   )
   &AddVelocityInCone(
    Cone Angle=20
    Cone Axis=<0 -1 0>
    Module.Use Velocity Falloff On Cone Axis=false
    Velocity Distribution Along Cone Axis=0.75
    Velocity Strength=&RandomRangeFloat(Minimum=500, Maximum=850)
   ) {enabled=false}
   &SubUVAnimation(End Frame=3, SubUV Animation Mode:ENiagara_SubUVLookupMode=::NewEnumerator2)
   &SkeletalMeshLocation(MaskSpawnGroup=false, OverrideRandomSeed=false, Skeletal Mesh=$User.SkeletalMesh) {enabled=false}
   &InheritVelocity(Inherited Velocity Amount Scale=<0.35 0.35 0.35>, Speed Limit=false, UseSpeedThreshold=false) {enabled=false}
  ]
  particleUpdateScript [
   &ParticleState()
   &GravityForce(Gravity=<0 0 -250>) {enabled=false}
   &Drag(Drag=10, Use Linear Drag=true, Use Rotational Drag=false)
   &ScaleColor(
    Scale Alpha=&FloatFromCurve(FloatCurve=DataInterface() {propertyValues {Curve {keys [
         {interpMode=RCIM_Cubic, tangentMode=RCTM_User, time=0, value=0}
         {interpMode=RCIM_Cubic, tangentMode=RCTM_User, time=0.5, value=1}
         {tangentMode=RCTM_User, time=1, value=0}
        ]}}})
    Scale RGB=&VectorFromFloat(Value=1)
    ScaleA=true
    ScaleRGB=true
   )
   &CurlNoiseForce(Noise Frequency=125, Noise Quality / Cost=::NewEnumerator0, Noise Strength=0.1, Pan Noise=true, Pan Noise Field=<0 0 -1>)
   &SolveForcesAndVelocity(Clamp Velocity=false, Limit Acceleration=false)
   &SpriteRotationRate(Rotation Rate=&RandomRangeFloat(Minimum=10, Maximum=-10)) {module=SpriteRotationRate001}
   &ScaleSpriteSize(
    Uniform Curve Scale=1.5
    Uniform Curve Sprite Scale=DataInterface() {propertyValues {Curve {keys [<at 0 0.75 RCIM_Cubic>,<at 1 1 RCIM_Cubic>]}}}
   )
  ]
  rendererClasses [&NiagaraSpriteRendererProperties]
  renderers [{rendererClass=&NiagaraSpriteRendererProperties}]
 }
 PolkaDotsSpawn_Pos {
  emitterUpdateScript [
   &EmitterState()
   &SpawnRate(SpawnRate=0.3, Use Spawn Probability=false)
   &SpawnBurst_Instantaneous(Use Spawn Probability=false, UseLoopCountLimit=false)
  ]
  particleSpawnScript [
   &InitializeParticle(
    Color Mode=::NewEnumerator0
    Lifetime=0.1
    Mass Max=0.0025
    Mass Min=0.01
    Mass Mode=::NewEnumerator3
    Sprite Rotation Mode=::NewEnumerator2
    Sprite Size Mode=::NewEnumerator3
    Uniform Sprite Size=0
    UsePositionOffset=true
    Write Material Random=true
   )
   &ShapeLocation(
    Cylinder Height=30
    Cylinder Height Midpoint=1
    Cylinder Radius=10
    Enable Lathe Profile=true
    Lathe Profile=DataInterface() {propertyValues {Curve {keys [{tangentMode=RCTM_User, time=0, value=0.5}, {tangentMode=RCTM_User, time=1.1, value=1.25}]}}}
    MaskSpawnGroup=false
    Override Seed=false
    Rotation Mode=::NewEnumerator1
    Shape Primitive=::NewEnumerator1
    Surface Only=false
    UseOwnerScale=true
   )
  ]
  particleUpdateScript [
   &ParticleState()
   &SolveForcesAndVelocity(Clamp Velocity=false, Limit Acceleration=false)
   &GenerateDeathEvent(Use Event Probability=false)
  ]
  rendererClasses [&NiagaraSpriteRendererProperties]
  renderers [{rendererClass=&NiagaraSpriteRendererProperties}]
 }
 SnowFlakes {
  emitterUpdateScript [
   &EmitterState()
   &SpawnRate(SpawnRate=2, Use Spawn Probability=false)
   &SpawnBurst_Instantaneous(Use Spawn Probability=false, UseLoopCountLimit=false)
  ]
  particleSpawnScript [
   &InitializeParticle(
    Color=&MakeCustomLinearColorFromBool(
     Bool=&RandomBool()
     True Color=<rgba 0.143 0.1568 0.1615 1>
     False Color=&RandomRangeLinearColor(Minimum=<rgba 0.1 0.655 1 0.1>, Maximum=<rgba 0.25 0.8375 1 0.1>)
    )
    Lifetime Min=1.5
    Lifetime Mode=::NewEnumerator1
    Mass Max=0.0025
    Mass Min=0.01
    Mass Mode=::NewEnumerator3
    Sprite Rotation Mode=::NewEnumerator2
    Sprite Size Mode=::NewEnumerator1
    Uniform Sprite Size Max=4.25
    Uniform Sprite Size Min=2.5
    UsePositionOffset=true
    Write Material Random=true
   )
   &Color(Color=&FN_Select_Color_from_Time_Of_Day(
     Morning=&HueShiftLinearColor(Hue=0, Saturation=1, Value=0.5, Linear Color=$Particles.Initial.Color)
     Day=$Particles.Initial.Color
     Evening=&HueShiftLinearColor(Hue=0, Saturation=1, Value=0.25, Linear Color=$Particles.Initial.Color)
     Night=&HueShiftLinearColor(Hue=0, Saturation=0.6, Value=0.95, Linear Color=$Particles.Initial.Color)
    ))
   &SphereLocation(
    MaskSpawnGroup=false
    Offset=<-20 0 5>
    Override Seed=false
    Sphere Radius=4
    Surface Only=false
    UseNonUniformScale=false
    UseOffset=true
    UseSphereOrientation=false
   )
   &AddVelocityInCone(
    Cone Angle=20
    Cone Axis=<0 -1 0>
    Module.Use Velocity Falloff On Cone Axis=false
    Velocity Distribution Along Cone Axis=0.75
    Velocity Strength=&RandomRangeFloat(Minimum=500, Maximum=850)
   ) {enabled=false}
   &SubUVAnimation(End Frame=3, SubUV Animation Mode:ENiagara_SubUVLookupMode=::NewEnumerator2)
   &SkeletalMeshLocation(MaskSpawnGroup=false, OverrideRandomSeed=false, Skeletal Mesh=$User.SkeletalMesh) {enabled=false}
   &InheritVelocity(Speed Limit=false, UseSpeedThreshold=false)
  ]
  particleUpdateScript [
   &ParticleState()
   &GravityForce(Gravity=<0 0 -250>) {enabled=false}
   &Drag(Drag=10, Use Linear Drag=true, Use Rotational Drag=false)
   &ScaleColor(
    Scale Alpha=&FloatFromCurve(FloatCurve=DataInterface() {propertyValues {Curve {keys [<at -0.0046 1.0113 RCIM_Cubic>,<at 1.0011 0.002 RCIM_Cubic>]}}})
    ScaleA=true
    ScaleRGB=true
   )
   &CurlNoiseForce(Noise Frequency=25, Noise Strength=1, Pan Noise=true, Pan Noise Field=<0 0 -1>)
   &SolveForcesAndVelocity(Clamp Velocity=false, Limit Acceleration=false)
   &SpriteRotationRate(Rotation Rate=&RandomRangeFloat(Minimum=250, Maximum=-250)) {module=SpriteRotationRate001}
   &ScaleSpriteSize(Uniform Curve Scale=2, Uniform Curve Sprite Scale=DataInterface() {propertyValues {Curve {keys [<at 0 0 RCIM_Cubic>,<at 1 1 RCIM_Cubic>]}}})
  ]
  rendererClasses [&NiagaraSpriteRendererProperties]
  renderers [{rendererClass=&NiagaraSpriteRendererProperties}]
 }
 SoftFire_Outside {
  emitterUpdateScript [
   &EmitterState()
   &SpawnRate(SpawnRate=5, Use Spawn Probability=false)
  ]
  particleSpawnScript [
   /Niagara/Modules/Spawn/Initialization/InitializeParticle(
    Color=&MultiplyLinearColor(Linear Color=$User.Primary Color, Scale Factor=<1 1 1 0.075>)
    Lifetime=4
    Module.Write Color=true
    Module.Write Lifetime=true
    Module.Write Mass=true
    Module.Write Material Random=true
    Module.Write Position=false
    Module.Write Scale=false
    Module.Write SpriteRotation=true
    Module.Write SpriteSize=true
    Position=&SimulationPosition()
    Sprite Rotation=&UniformRangedFloat(Minimum=0, Maximum=360)
    Sprite Size=&UniformRangedVector2D(Minimum=<25 25>, Maximum=<35 35>)
   )
   &SystemLocation(MaskSpawnGroup=false, Offset=<20 0 0>)
   &SphereLocation(
    MaskSpawnGroup=false
    Override Seed=false
    Sphere Radius=6
    Surface Only=false
    UseNonUniformScale=false
    UseOffset=false
    UseSphereOrientation=false
   )
   &AddVelocity(Velocity=&UniformRangedVector(Minimum=<-10 0 0>, Maximum=<-15 0 0>))
   /SetVariables(Particles.RandomStartingFrame:NiagaraFloat=&UniformRangedFloat(Minimum=0, Maximum=1)) {isSetParameters=true, module=SetVariables_#1}
   /SetVariables(Particles.InitialRotationRate:NiagaraFloat=&UniformRangedFloat(Minimum=-60, Maximum=0)) {isSetParameters=true, module=SetVariables_#2}
   &CameraOffset(Camera Offset Amount=-10) {module=CameraOffset001}
   &Color(Color=&FN_Select_Color_from_Time_Of_Day(
     Morning=&HueShiftLinearColor(Hue=0, Saturation=1, Value=0.5, Linear Color=$Particles.Initial.Color)
     Day=$Particles.Initial.Color
     Evening=&HueShiftLinearColor(Hue=0, Saturation=1, Value=0.25, Linear Color=$Particles.Initial.Color)
     Night=&HueShiftLinearColor(Hue=0, Saturation=0.25, Value=0.9, Linear Color=$Particles.Initial.Color)
    ))
  ]
  particleUpdateScript [
   &ParticleState()
   /Niagara/Modules/Update/Velocity/VortexVelocity(Module.Use Influence Falloff=false, Velocity Amount=30, Vortex Axis:Vector3f=<1 0 0>)
   /Niagara/Modules/Update/Forces/PointAttractionForce(
    Attraction Radius=65
    AttractionStrength=75
    AttractorPosition:Vector3f=&Add_Vector(A=/Niagara/DynamicInputs/Helpers/SimulationPosition(), B=<-40 0 0>)
    Module.Kill Within Radius=false
    Module.Use Falloff=true
   )
   &SolveForcesAndVelocity(Clamp Velocity=false, Limit Acceleration=false)
   &ScaleColor(
    Scale Alpha=&FloatFromCurve(
     FloatCurve=DataInterface() {propertyValues {Curve {keys [
         {interpMode=RCIM_Cubic, leaveTangent=8.9046, tangentMode=RCTM_Break, time=0, value=0}
         <at 0.1192 0.9901>
         {interpMode=RCIM_Cubic, tangentMode=RCTM_User, time=0.75, value=1}
         {interpMode=RCIM_Cubic, tangentMode=RCTM_Break, time=1, value=0}
        ]}}}
     Scale Curve=0.2
    )
    Scale RGB=<6 6 6>
    ScaleA=true
    ScaleRGB=true
   )
   &CameraOffset(Camera Offset Amount=&FloatFromCurve(FloatCurve=DataInterface() {propertyValues {Curve {keys [<at 0.0099 23.4541>,<at 0.1387 27.7338>,<at 0.612 17.5534>,<at 0.7977 21.7682>,<at 1 20>]}}}))
   &ScaleSpriteSize(Scale Factor=&Vector2DFromFloat(Value=&FloatFromCurve(FloatCurve=DataInterface() {propertyValues {Curve {keys [
          {arriveTangent=4.4129, interpMode=RCIM_Cubic, leaveTangent=4.4129, tangentMode=RCTM_User, time=0, value=0.5}
          {
           arriveTangentWeight=0.7777
           interpMode=RCIM_Cubic
           leaveTangentWeight=0.7777
           tangentMode=RCTM_User
           tangentWeightMode=RCTWM_WeightedBoth
           time=0.1967
           value=1
          }
          {arriveTangent=-2.8109, interpMode=RCIM_Cubic, leaveTangent=-2.8109, tangentMode=RCTM_User, time=1, value=0}
         ]}}})))
   /Niagara/Modules/Update/SubUV/SubUVAnimation(
    Number Of Frames=64
    SubUV Lookup Index=&Multiply_Float(A=&Add_Float(A=$Particles.NormalizedAge, B=$Particles.RandomStartingFrame), B=0.6)
   )
   &SpriteRotationRate(Rotation Rate=$Particles.InitialRotationRate)
   &DynamicMaterialParameters(
    Index 0 Param 1=&FloatFromCurve(FloatCurve=DataInterface() {propertyValues {Curve {keys [<at 0 0.25 RCIM_Cubic>,<at 1 -0.1>]}}}, CurveIndex=$Particles.NormalizedLife)
    Module.Param0WriteEnabled=true
    Module.Param1WriteEnabled=false
    Module.Param2WriteEnabled=false
    Module.Param3WriteEnabled=false
   )
  ]
  rendererClasses [&NiagaraSpriteRendererProperties]
  renderers [{rendererClass=&NiagaraSpriteRendererProperties}]
 }
 SoftFire_Outside001 {
  emitterUpdateScript [
   &EmitterState(Life Cycle Mode=::NewEnumerator1, Loop Behavior=::NewEnumerator1, Loop Duration=1, UseLoopDelay=false)
   &SpawnBurst_Instantaneous(Spawn Count=5, Use Spawn Probability=false, UseLoopCountLimit=false)
   &SpawnBurst_Instantaneous(Spawn Count=3, Spawn Group=1, Use Spawn Probability=false, UseLoopCountLimit=false) {enabled=false, module=SpawnBurst_Instantaneous001}
  ]
  particleSpawnScript [
   &InitializeParticle(
    Color=&MultiplyLinearColor(Linear Color=$User.Primary Color, Scale Factor=<1 1 1 0.1>)
    Lifetime=0.75
    Module.Write Color=true
    Module.Write Lifetime=true
    Module.Write Mass=true
    Module.Write Material Random=true
    Module.Write Position=false
    Module.Write Scale=false
    Module.Write SpriteRotation=true
    Module.Write SpriteSize=true
    Position=/Niagara/Functions/Localspace/SimulationPosition()
    Sprite Rotation=&RandomRangeFloat(Minimum=0, Maximum=360)
    Sprite Size=&UniformRangedVector2D(Minimum=<25 25>, Maximum=<35 35>)
   )
   &SystemLocation(MaskSpawnGroup=false)
   &SphereLocation(
    MaskSpawnGroup=false
    Override Seed=false
    Sphere Radius=6
    Surface Only=false
    UseNonUniformScale=false
    UseOffset=false
    UseSphereOrientation=false
   )
   /SetVariables(Particles.RandomStartingFrame:NiagaraFloat=&UniformRangedFloat(Minimum=0, Maximum=1)) {isSetParameters=true, module=SetVariables_#1}
   /SetVariables(Particles.InitialRotationRate:NiagaraFloat=&UniformRangedFloat(Minimum=-60, Maximum=0)) {isSetParameters=true, module=SetVariables_#2}
   &CameraOffset(Camera Offset Amount=-10) {module=CameraOffset001}
   &Color(Color=&FN_Select_Color_from_Time_Of_Day(
     Morning=&HueShiftLinearColor(Hue=0, Saturation=1, Value=0.5, Linear Color=$Particles.Initial.Color)
     Day=$Particles.Initial.Color
     Evening=&HueShiftLinearColor(Hue=0, Saturation=1, Value=0.25, Linear Color=$Particles.Initial.Color)
     Night=&HueShiftLinearColor(Hue=0, Saturation=0.25, Value=0.9, Linear Color=$Particles.Initial.Color)
    ))
  ]
  particleUpdateScript [
   &ParticleState()
   &VortexVelocity(Module.Use Influence Falloff=false, Velocity Amount=30, Vortex Axis:Vector3f=<1 0 0>) {enabled=false}
   &PointAttractionForce(
    Attraction Radius=65
    AttractionStrength=75
    AttractorPosition:Vector3f=&Add_Vector(A=/Niagara/DynamicInputs/Helpers/SimulationPosition(), B=<-40 0 0>)
    Module.Kill Within Radius=false
    Module.Use Falloff=true
   ) {enabled=false}
   &ScaleColor(
    Scale Alpha=&FloatFromCurve(
     FloatCurve=DataInterface() {propertyValues {Curve {keys [
         {interpMode=RCIM_Cubic, leaveTangent=5.8676, tangentMode=RCTM_Break, time=0, value=0}
         {
          arriveTangentWeight=0.0585
          interpMode=RCIM_Cubic
          leaveTangentWeight=0.4306
          tangentMode=RCTM_Break
          tangentWeightMode=RCTWM_WeightedBoth
          time=0.1
          value=1
         }
         {interpMode=RCIM_Cubic, tangentMode=RCTM_Break, time=1, value=0}
        ]}}}
     Scale Curve=0.2
    )
    Scale RGB=<6 6 6>
    ScaleA=true
    ScaleRGB=true
   )
   &CameraOffset(Camera Offset Amount=&FloatFromCurve(FloatCurve=DataInterface() {propertyValues {Curve {keys [<at 0.0099 23.4541>,<at 0.1387 27.7338>,<at 0.612 17.5534>,<at 0.7977 21.7682>,<at 1 20>]}}}))
   &SpriteRotationRate(Rotation Rate=$Particles.InitialRotationRate)
   &DynamicMaterialParameters(
    Index 0 Param 1=&FloatFromCurve(FloatCurve=DataInterface() {propertyValues {Curve {keys [<at 0 0.25 RCIM_Cubic>,<at 1 -0.1>]}}}, CurveIndex=$Particles.NormalizedLife)
    Module.Param0WriteEnabled=true
    Module.Param1WriteEnabled=false
    Module.Param2WriteEnabled=false
    Module.Param3WriteEnabled=false
   )
   /SetVariables(Particles.Position=&Multiply_Vector(
     A=&Lerp_Vector(A=<0 0 0>, B=<-60 0 0>, Alpha=$System.Timeline.NormalizedTime)
     B=&VectorFromFloat(Value=&ReturnNormalizedExecIndex())
    )) {isSetParameters=true}
   /Niagara/Modules/Update/SubUV/V2/SubUVAnimation(SubUV Loop Count=0.5) {module=SubUVAnimation001}
   &ScaleSpriteSize(
    Uniform Curve Scale=1.5
    Uniform Curve Sprite Scale=DataInterface() {propertyValues {Curve {keys [<at 0 1 RCIM_Cubic>, {arriveTangent=-2.0967, interpMode=RCIM_Cubic, leaveTangent=-2.0967, tangentMode=RCTM_User, time=1, value=0}]}}}
   )
  ]
  rendererClasses [&NiagaraSpriteRendererProperties]
  renderers [{rendererClass=&NiagaraSpriteRendererProperties}]
 }
 burst_flakes {
  emitterUpdateScript [
   &EmitterState(Life Cycle Mode=::NewEnumerator1, Loop Behavior=::NewEnumerator1, Loop Duration=0.5, UseLoopDelay=false)
   &SpawnRate(SpawnRate=30, Use Spawn Probability=false)
  ]
  particleSpawnScript [
   /Niagara/Modules/Spawn/Initialization/V2/InitializeParticle(
    Color=<rgba 0.0886 0.3126 0.5906 0.1>
    Lifetime Max=0.25
    Lifetime Min=0.5
    Lifetime Mode=::NewEnumerator1
    Mass Mode=::NewEnumerator3
    Sprite Rotation Mode=::NewEnumerator2
    Sprite Size Mode=::NewEnumerator1
    Uniform Sprite Size Max=2
    Uniform Sprite Size Min=3
    UsePositionOffset=false
    Write Material Random=true
   )
   &SphereLocation(
    MaskSpawnGroup=false
    Override Seed=false
    Sphere Radius=8
    Surface Only=false
    UseNonUniformScale=false
    UseOffset=false
    UseSphereOrientation=false
   )
   /SetVariables() {isSetParameters=true}
   &AddVelocityInCone(
    Cone Angle=&Add_Float(A=180, B=&RandomRangeFloat(Minimum=-20, Maximum=20))
    Module.Use Velocity Falloff On Cone Axis=false
    Velocity Distribution Along Cone Axis=0.75
    Velocity Strength=&Multiply_Float(
     A=&RandomRangeFloat(Minimum=100, Maximum=300)
     B=&FloatFromCurve(FloatCurve=DataInterface() {propertyValues {Curve {keys [<at 0 2>,<at 1 1>]}}}, CurveIndex=$System.Timeline.NormalizedTime)
    )
   )
   &SubUVAnimation(End Frame=3, SubUV Animation Mode:ENiagara_SubUVLookupMode=::NewEnumerator2)
  ]
  particleUpdateScript [
   &ParticleState()
   &GravityForce() {enabled=false}
   &CurlNoiseForce(Noise Frequency=120, Noise Strength=250, Pan Noise=false)
   &Drag(Drag=15, Use Linear Drag=true, Use Rotational Drag=false)
   &ScaleColor(
    Scale Alpha=&FloatFromCurve(FloatCurve=DataInterface() {propertyValues {Curve {keys [
         <at -0.0046 1.0113 RCIM_Cubic>
         {arriveTangent=-2.0133, interpMode=RCIM_Cubic, leaveTangent=-2.0133, tangentMode=RCTM_User, time=1.0011, value=0.002}
        ]}}})
    ScaleA=true
    ScaleRGB=true
   )
   &SolveForcesAndVelocity(Clamp Velocity=false, Limit Acceleration=false)
   &ScaleSpriteSize(
    Uniform Curve Scale=2
    Uniform Curve Sprite Scale=DataInterface() {propertyValues {Curve {keys [
        {interpMode=RCIM_Cubic, tangentMode=RCTM_User, time=0, value=1}
        {arriveTangent=-2.729, interpMode=RCIM_Cubic, leaveTangent=-2.729, tangentMode=RCTM_User, time=1, value=0}
       ]}}}
   )
  ]
  rendererClasses [&NiagaraSpriteRendererProperties]
  renderers [
   {rendererClass=&NiagaraSpriteRendererProperties}
   {rendererClass=&NiagaraSpriteRendererProperties, rendererIndex=1}
  ]
 }
 cresaunt {
  emitterUpdateScript [
   &EmitterState()
   &SpawnBurst_Instantaneous(Age:NiagaraFloat=$Emitter.Age, Use Spawn Probability=false, UseLoopCountLimit=false)
  ]
  particleSpawnScript [
   &InitializeParticle(
    Color=<rgba 0.2104 0.5616 0.765 1>
    Lifetime=5
    Mesh Scale Mode=::NewEnumerator3
    Mesh Uniform Scale=0.65
    Position Offset=<2 0 0>
    Sprite Size Mode=::NewEnumerator3
    Uniform Sprite Size=45
    UsePositionOffset=true
    Write Material Random=true
   )
   &InitialMeshOrientation(Mesh Orientation Mode=::NewEnumerator7, Rotation=<0 0.25 0>, Use Rotation Vector=true)
   /SetVariables(
    Particles.Sprite cam Offset=-35
    Particles.Sprite color:LinearColor=&FN_Select_Color_from_Time_Of_Day(
     Morning=&HueShiftLinearColor(Hue=0, Saturation=1, Value=0.5, Linear Color=$Particles.Initial.Color)
     Day=$Particles.Initial.Color
     Evening=&HueShiftLinearColor(Hue=0, Saturation=1, Value=0.25, Linear Color=$Particles.Initial.Color)
     Night=&HueShiftLinearColor(Hue=0, Saturation=0.6, Value=1, Linear Color=$Particles.Initial.Color)
    )
   ) {isSetParameters=true}
   /SetVariables(
    Particles.Sprite cam Offset=-35
    Particles.Sprite color:LinearColor=&MultiplyLinearColor(Linear Color=$Particles.Sprite color, Scale Factor=<1 1 1 0.12>)
   ) {isSetParameters=true}
   &InitialMeshOrientation(Mesh Orientation Mode=::NewEnumerator1, Rotation=<-0.25 0 0>, Use Rotation Vector=true) {module=InitialMeshOrientation001}
   /SetVariables(Particles.inner color:LinearColor=<rgba 0.1 0.2189 0.25 1>) {isSetParameters=true}
  ]
  particleUpdateScript [
   &ParticleState(
    Kill Particles When Lifetime Has Elapsed=false
    Let Infinitely Lived Particles Die When Emitter Deactivates=true
    Loop Particles Lifetime=true
   )
   &Color(Color=&FN_Select_Color_from_Time_Of_Day(
     Morning=&HueShiftLinearColor(Hue=0, Saturation=1, Value=0.5, Linear Color=$Particles.Initial.Color)
     Day=$Particles.Initial.Color
     Evening=&HueShiftLinearColor(Hue=0, Saturation=1, Value=0.25, Linear Color=$Particles.Initial.Color)
     Night=&HueShiftLinearColor(Hue=0, Saturation=0.25, Value=0.85, Linear Color=$Particles.Initial.Color)
    ))
   &UpdateMeshOrientation(Rotation Rate=0.025, Rotation Vector=<0 0 1>)
   &FN_Light_Attributes(
    Color=&FN_Select_Color_from_Time_Of_Day(
     Morning=&HueShiftLinearColor(Hue=0, Saturation=1, Value=0.5, Linear Color=$Particles.Initial.inner color)
     Day=$Particles.Initial.inner color
     Evening=&HueShiftLinearColor(Hue=0, Saturation=1, Value=0.25, Linear Color=$Particles.Initial.inner color)
     Night=&HueShiftLinearColor(Hue=0, Saturation=0.6, Value=1, Linear Color=$Particles.Initial.inner color)
    )
    Create Color Attribute=true
    Create Exponent Attribute=true
    Create Radius Attribute=true
    Exponent=3
    Radius=45
    Scale Color=3
   )
   /SetVariables(Particles.DynamicMaterialParameter:Vector4f=&MakeVector4_FromLinearColor(LinearColor=<rgba 0.1 0.2189 0.25 1>)) {enabled=false, isSetParameters=true}
   /SetVariables(Particles.DynamicMaterialParameter:Vector4f=&MakeVector4_FromLinearColor(LinearColor=&FN_Select_Color_from_Time_Of_Day(
      Morning=&HueShiftLinearColor(Hue=0, Saturation=1, Value=0.5, Linear Color=$Particles.Initial.inner color)
      Day=$Particles.Initial.inner color
      Evening=&HueShiftLinearColor(Hue=0, Saturation=1, Value=0.25, Linear Color=$Particles.Initial.inner color)
      Night=&HueShiftLinearColor(Hue=0, Saturation=0.6, Value=1, Linear Color=$Particles.Initial.inner color)
     ))) {isSetParameters=true}
   &ScaleMeshSize(Scale Factor=&VectorFromFloat(Value=&FloatFromCurve(
      FloatCurve=DataInterface() {propertyValues {Curve {keys [{arriveTangent=4.1976, interpMode=RCIM_Cubic, time=0, value=0}, {tangentMode=RCTM_User, time=0.75, value=1}]}}}
      CurveIndex=$System.Timeline.NormalizedTime
     )))
  ]
  rendererClasses [
   &NiagaraMeshRendererProperties
   &NiagaraSpriteRendererProperties
   &NiagaraLightRendererProperties
  ]
  renderers [
   {rendererClass=&NiagaraMeshRendererProperties, rendererIndex=0}
   {rendererClass=&NiagaraSpriteRendererProperties, rendererIndex=1}
   {rendererClass=&NiagaraLightRendererProperties, rendererIndex=2}
  ]
 }
}
summary {
 systemName=NS_Pickaxe_SnowFall
 emitters [
  {emitterName=SoftFire_Outside, rendererClasses [&NiagaraSpriteRendererProperties]}
  {emitterName=Flame_MAIN, rendererClasses [&NiagaraMeshRendererProperties]}
  {emitterName=LightGlow, rendererClasses [&NiagaraSpriteRendererProperties, &NiagaraLightRendererProperties]}
  {emitterName=Decal_Glow_Wrist, rendererClasses [&NiagaraSpriteRendererProperties]}
  {emitterName=SnowFlakes, rendererClasses [&NiagaraSpriteRendererProperties]}
  {emitterName=Clouds, rendererClasses [&NiagaraSpriteRendererProperties, &NiagaraMeshRendererProperties]}
  {
   emitterName=cresaunt
   rendererClasses [
    &NiagaraMeshRendererProperties
    &NiagaraSpriteRendererProperties
    &NiagaraLightRendererProperties
   ]
  }
  {emitterName=PolkaDotsSpawn_Pos, rendererClasses [&NiagaraSpriteRendererProperties]}
  {emitterName=PolkaDots001, rendererClasses [&NiagaraSpriteRendererProperties]}
  {emitterName=SoftFire_Outside001, rendererClasses [&NiagaraSpriteRendererProperties]}
  {emitterName=Cold_Smoke, rendererClasses [&NiagaraSpriteRendererProperties]}
  {emitterName=burst_flakes, rendererClasses [&NiagaraSpriteRendererProperties]}
  {emitterName=Fountain002, rendererClasses [&NiagaraMeshRendererProperties]}
 ]
 userVariables [
  User.Primary Color {type=LinearColor, description="", defaultValue {struct=LinearColor, value=<rgba 0.3755 0.822 0.9387 1>}}
  User.Secondary Color {type=LinearColor, description="", defaultValue {struct=LinearColor, value=<rgba 0.0886 0.3126 0.5906 1>}}
 ]
}
